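# Sound cue for the Zero Point Sprite's giggle emotion voice line, spatialized with close VO attenuation and limited by a sprite music-vox concurrency set.
# A root Modulator randomizes playback, with a widened volume range, over a Quality Level switch feeding two Random pools of Patrol_Giggle waves.
# The first two quality slots share a nine-wave pool excluded from branch culling, while the last two share a four-wave subset of it.
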SoundCue /Sprites_ZeroPointSprite/Assets/Sounds/Cues/VO/SC_VO_S33_ZeroPointSprite_Emotion_Giggle
root=Modulator_0
nodes [
 {id=Modulator_0, children [QualityLevel_0], properties {volumeMin=0.6}}
 {id=QualityLevel_0, children [Random_1,Random_1,Random_2,Random_2]}
 {
  id=Random_1
  children [WavePlayer_5, WavePlayer_6, WavePlayer_7, WavePlayer_8, WavePlayer_9, WavePlayer_10, WavePlayer_11, WavePlayer_12, WavePlayer_13]
  properties {bShouldExcludeFromBranchCulling=true, weights [1,1,1,1,1,1,1,1,1]}
 }
 {id=Random_2, children [WavePlayer_14, WavePlayer_15, WavePlayer_16, WavePlayer_17], properties {weights [1,1,1,1]}}
 {id=WavePlayer_10, properties {soundWaveAssetPtr=/Sprites_ZeroPointSprite/Assets/Sounds/Waves/VO/VO_S33_ZeroPointSprite_Patrol_Giggle_012}}
 {id=WavePlayer_11, properties {soundWaveAssetPtr=/Sprites_ZeroPointSprite/Assets/Sounds/Waves/VO/VO_S33_ZeroPointSprite_Patrol_Giggle_013}}
 {id=WavePlayer_12, properties {soundWaveAssetPtr=/Sprites_ZeroPointSprite/Assets/Sounds/Waves/VO/VO_S33_ZeroPointSprite_Patrol_Giggle_014}}
 {id=WavePlayer_13, properties {soundWaveAssetPtr=/Sprites_ZeroPointSprite/Assets/Sounds/Waves/VO/VO_S33_ZeroPointSprite_Patrol_Giggle_015}}
 {id=WavePlayer_14, properties {soundWaveAssetPtr=/Sprites_ZeroPointSprite/Assets/Sounds/Waves/VO/VO_S33_ZeroPointSprite_Patrol_Giggle_016}}
 {id=WavePlayer_15, properties {soundWaveAssetPtr=/Sprites_ZeroPointSprite/Assets/Sounds/Waves/VO/VO_S33_ZeroPointSprite_Patrol_Giggle_017}}
 {id=WavePlayer_16, properties {soundWaveAssetPtr=/Sprites_ZeroPointSprite/Assets/Sounds/Waves/VO/VO_S33_ZeroPointSprite_Patrol_Giggle_018}}
 {id=WavePlayer_17, properties {soundWaveAssetPtr=/Sprites_ZeroPointSprite/Assets/Sounds/Waves/VO/VO_S33_ZeroPointSprite_Patrol_Giggle_010}}
 {id=WavePlayer_5, properties {soundWaveAssetPtr=&VO_S33_ZeroPointSprite_Patrol_Giggle_016}}
 {id=WavePlayer_6, properties {soundWaveAssetPtr=&VO_S33_ZeroPointSprite_Patrol_Giggle_017}}
 {id=WavePlayer_7, properties {soundWaveAssetPtr=&VO_S33_ZeroPointSprite_Patrol_Giggle_018}}
 {id=WavePlayer_8, properties {soundWaveAssetPtr=&VO_S33_ZeroPointSprite_Patrol_Giggle_010}}
 {id=WavePlayer_9, properties {soundWaveAssetPtr=/Sprites_ZeroPointSprite/Assets/Sounds/Waves/VO/VO_S33_ZeroPointSprite_Patrol_Giggle_011}}
]
properties {
 attenuationSettings=/Sprites_ZeroPointSprite/Assets/Sounds/Settings/AS_Sprite_VO_Close
 maxDistance=4000
 soundClassObject=/Game/Sounds/FortSoundClassesAndModes/Fort_Creatures_Vocalization_Normal
 volumeMultiplier=1
 concurrencySet [/Sprites/Assets/Sounds/Music/CS_Sprite_Pri_MusicVox]
}
