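# Sound cue for the sad emotion voice of the Season 33 Zero Point Sprite creature.
# A quality-level node sends its first two levels to a random pick of nine takes, excluded from branch culling, and its last two to a four-take subset.
# A modulator above it varies volume.
# Close sprite VO attenuation, a creature vocalization sound class and a shared sprite concurrency set govern playback.
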SoundCue /Sprites_ZeroPointSprite/Assets/Sounds/Cues/VO/SC_VO_S33_ZeroPointSprite_Emotion_Sad
root=Modulator_0
nodes [
 {id=Modulator_0, children [QualityLevel_0], properties {volumeMin=0.6}}
 {id=QualityLevel_0, children [Random_0,Random_0,Random_1,Random_1]}
 {
  id=Random_0
  children [WavePlayer_0,WavePlayer_1,WavePlayer_2,WavePlayer_3,WavePlayer_4,WavePlayer_5,WavePlayer_6,WavePlayer_7,WavePlayer_8]
  properties {bShouldExcludeFromBranchCulling=true, weights [1,1,1,1,1,1,1,1,1]}
 }
 {id=Random_1, children [WavePlayer_9, WavePlayer_10, WavePlayer_11, WavePlayer_12], properties {weights [1,1,1,1]}}
 {id=WavePlayer_0, properties {soundWaveAssetPtr=/Sprites_ZeroPointSprite/Assets/Sounds/Waves/VO/VO_S33_ZeroPointSprite_Emotion_Sad_010}}
 {id=WavePlayer_1, properties {soundWaveAssetPtr=/Sprites_ZeroPointSprite/Assets/Sounds/Waves/VO/VO_S33_ZeroPointSprite_Emotion_Sad_011}}
 {id=WavePlayer_10, properties {soundWaveAssetPtr=/Sprites_ZeroPointSprite/Assets/Sounds/Waves/VO/VO_S33_ZeroPointSprite_Emotion_Sad_013}}
 {id=WavePlayer_11, properties {soundWaveAssetPtr=/Sprites_ZeroPointSprite/Assets/Sounds/Waves/VO/VO_S33_ZeroPointSprite_Emotion_Sad_015}}
 {id=WavePlayer_12, properties {soundWaveAssetPtr=/Sprites_ZeroPointSprite/Assets/Sounds/Waves/VO/VO_S33_ZeroPointSprite_Emotion_Sad_017}}
 {id=WavePlayer_2, properties {soundWaveAssetPtr=/Sprites_ZeroPointSprite/Assets/Sounds/Waves/VO/VO_S33_ZeroPointSprite_Emotion_Sad_012}}
 {id=WavePlayer_3, properties {soundWaveAssetPtr=&VO_S33_ZeroPointSprite_Emotion_Sad_013}}
 {id=WavePlayer_4, properties {soundWaveAssetPtr=/Sprites_ZeroPointSprite/Assets/Sounds/Waves/VO/VO_S33_ZeroPointSprite_Emotion_Sad_014}}
 {id=WavePlayer_5, properties {soundWaveAssetPtr=&VO_S33_ZeroPointSprite_Emotion_Sad_015}}
 {id=WavePlayer_6, properties {soundWaveAssetPtr=/Sprites_ZeroPointSprite/Assets/Sounds/Waves/VO/VO_S33_ZeroPointSprite_Emotion_Sad_016}}
 {id=WavePlayer_7, properties {soundWaveAssetPtr=&VO_S33_ZeroPointSprite_Emotion_Sad_017}}
 {id=WavePlayer_8, properties {soundWaveAssetPtr=/Sprites_ZeroPointSprite/Assets/Sounds/Waves/VO/VO_S33_ZeroPointSprite_Emotion_Sad_018}}
 {id=WavePlayer_9, properties {soundWaveAssetPtr=&VO_S33_ZeroPointSprite_Emotion_Sad_010}}
]
properties {
 attenuationSettings=/Sprites_ZeroPointSprite/Assets/Sounds/Settings/AS_Sprite_VO_Close
 maxDistance=4000
 soundClassObject=/Game/Sounds/FortSoundClassesAndModes/Fort_Creatures_Vocalization_Normal
 volumeMultiplier=1
 concurrencySet [/Sprites/Assets/Sounds/Music/CS_Sprite_Pri_MusicVox]
}
