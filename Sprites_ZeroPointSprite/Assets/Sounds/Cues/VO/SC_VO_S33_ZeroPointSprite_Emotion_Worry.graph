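# Sound cue for the Zero Point Sprite's worry emotion vocalization.
# A Random node picks one of five equally weighted voice waves, excluded from branch culling, and feeds a Modulator that varies volume with a lowered minimum.
# The cue routes to the creature vocalization sound class, uses close-range sprite VO attenuation, and shares a sprite music-vox concurrency set.
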SoundCue /Sprites_ZeroPointSprite/Assets/Sounds/Cues/VO/SC_VO_S33_ZeroPointSprite_Emotion_Worry
root=Modulator_0
nodes [
 {id=Modulator_0, children [Random_0], properties {volumeMin=0.6}}
 {
  id=Random_0
  children [WavePlayer_0,WavePlayer_1,WavePlayer_2,WavePlayer_3,WavePlayer_4]
  properties {bShouldExcludeFromBranchCulling=true, weights [1,1,1,1,1]}
 }
 {id=WavePlayer_0, properties {soundWaveAssetPtr=/Sprites_ZeroPointSprite/Assets/Sounds/Waves/VO/VO_S33_ZeroPointSprite_Emotion_Worry_010}}
 {id=WavePlayer_1, properties {soundWaveAssetPtr=/Sprites_ZeroPointSprite/Assets/Sounds/Waves/VO/VO_S33_ZeroPointSprite_Emotion_Worry_011}}
 {id=WavePlayer_2, properties {soundWaveAssetPtr=/Sprites_ZeroPointSprite/Assets/Sounds/Waves/VO/VO_S33_ZeroPointSprite_Emotion_Worry_012}}
 {id=WavePlayer_3, properties {soundWaveAssetPtr=/Sprites_ZeroPointSprite/Assets/Sounds/Waves/VO/VO_S33_ZeroPointSprite_Emotion_Worry_013}}
 {id=WavePlayer_4, properties {soundWaveAssetPtr=/Sprites_ZeroPointSprite/Assets/Sounds/Waves/VO/VO_S33_ZeroPointSprite_Emotion_Worry_014}}
]
properties {
 attenuationSettings=/Sprites_ZeroPointSprite/Assets/Sounds/Settings/AS_Sprite_VO_Close
 maxDistance=4000
 soundClassObject=/Game/Sounds/FortSoundClassesAndModes/Fort_Creatures_Vocalization_Normal
 volumeMultiplier=1
 concurrencySet [/Sprites/Assets/Sounds/Music/CS_Sprite_Pri_MusicVox]
}
