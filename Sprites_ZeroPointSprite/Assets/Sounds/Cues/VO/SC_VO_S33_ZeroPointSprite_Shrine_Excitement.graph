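# Sound cue for the Zero Point Sprite's excited vocalizations, named for its shrine context.
# A random node picks one of five equally weighted voice takes and is excluded from branch culling.
# The pick feeds a modulator for slight pitch and volume variation.
# The cue uses far sprite VO attenuation, a creature vocalization sound class, a shared music-vox concurrency set and boosted volume.
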SoundCue /Sprites_ZeroPointSprite/Assets/Sounds/Cues/VO/SC_VO_S33_ZeroPointSprite_Shrine_Excitement
root=Modulator_0
nodes [
 {id=Modulator_0, children [Random_1], properties {volumeMin=1}}
 {
  id=Random_1
  children [WavePlayer_5,WavePlayer_6,WavePlayer_7,WavePlayer_8,WavePlayer_9]
  properties {bShouldExcludeFromBranchCulling=true, weights [1,1,1,1,1]}
 }
 {id=WavePlayer_5, properties {soundWaveAssetPtr=/Sprites_ZeroPointSprite/Assets/Sounds/Waves/VO/VO_S33_ZeroPointSprite_Shrine_Excitement_010}}
 {id=WavePlayer_6, properties {soundWaveAssetPtr=/Sprites_ZeroPointSprite/Assets/Sounds/Waves/VO/VO_S33_ZeroPointSprite_Shrine_Excitement_011}}
 {id=WavePlayer_7, properties {soundWaveAssetPtr=/Sprites_ZeroPointSprite/Assets/Sounds/Waves/VO/VO_S33_ZeroPointSprite_Shrine_Excitement_012}}
 {id=WavePlayer_8, properties {soundWaveAssetPtr=/Sprites_ZeroPointSprite/Assets/Sounds/Waves/VO/VO_S33_ZeroPointSprite_Shrine_Excitement_013}}
 {id=WavePlayer_9, properties {soundWaveAssetPtr=/Sprites_ZeroPointSprite/Assets/Sounds/Waves/VO/VO_S33_ZeroPointSprite_Shrine_Excitement_014}}
]
properties {
 attenuationSettings=/Sprites_ZeroPointSprite/Assets/Sounds/Settings/AS_Sprite_VO_Far
 maxDistance=5400
 soundClassObject=&Fort_Creatures_Vocalization_Tactical
 volumeMultiplier=1.5
 concurrencySet [/Sprites/Assets/Sounds/Music/CS_Sprite_Pri_MusicVox]
}
