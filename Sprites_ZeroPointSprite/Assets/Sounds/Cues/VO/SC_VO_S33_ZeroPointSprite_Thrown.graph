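# Sound cue for throwing the Zero Point Sprite, combining its vocalization with throw effects and an aura loop.
# A mixer layers a random choice of five slightly delayed voice lines, one of three random throw sounds, and a quieter looping aura.
# The aura is pitched up and faded in by an envelope.
# It uses projectile in-air loop attenuation and the creature vocalization sound class.
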SoundCue /Sprites_ZeroPointSprite/Assets/Sounds/Cues/VO/SC_VO_S33_ZeroPointSprite_Thrown
root=Mixer_0
nodes [
 {id=Delay_0, children [Modulator_0], properties {delayMax=0.1, delayMin=0.1}}
 {
  id=Enveloper_0
  children [Modulator_1]
  properties {pitchCurve {editorCurveData {keys [{time=0}]}}, volumeCurve {editorCurveData {keys [<at 0 0.0048>,<at 0.5 1>]}}}
 }
 {id=Mixer_0, children [Delay_0,Enveloper_0,Modulator_2], properties {inputVolume [0.704,0.2,1]}}
 {id=Modulator_0, children [Random_1], properties {volumeMin=0.8}}
 {id=Modulator_1, children [WavePlayer_12], properties {pitchMax=2, pitchMin=2}}
 {id=Modulator_2, children [Random_3]}
 {
  id=Random_1
  children [WavePlayer_4,WavePlayer_5,WavePlayer_6,WavePlayer_7,WavePlayer_9]
  properties {bShouldExcludeFromBranchCulling=true, weights [1,1,1,1,1]}
 }
 {id=Random_3, children [WavePlayer_13, WavePlayer_14, WavePlayer_19], properties {weights [1,1,1]}}
 {id=WavePlayer_12, properties {bLooping=true, soundWaveAssetPtr=/Sprites_ZeroPointSprite/Assets/Sounds/Waves/SFX/SideKick_Sprite_aura_loop}}
 {id=WavePlayer_13, properties {soundWaveAssetPtr=/Sprites_ZeroPointSprite/Assets/Sounds/Waves/SFX/SideKick_Sprite_Throw_01}}
 {id=WavePlayer_14, properties {soundWaveAssetPtr=/Sprites_ZeroPointSprite/Assets/Sounds/Waves/SFX/SideKick_Sprite_Throw_02}}
 {id=WavePlayer_19, properties {soundWaveAssetPtr=/Sprites_ZeroPointSprite/Assets/Sounds/Waves/SFX/SideKick_Sprite_Throw_03}}
 {id=WavePlayer_4, properties {soundWaveAssetPtr=/Sprites_ZeroPointSprite/Assets/Sounds/Waves/VO/VO_S33_ZeroPointSprite_Inventory_Thrown_010}}
 {id=WavePlayer_5, properties {soundWaveAssetPtr=/Sprites_ZeroPointSprite/Assets/Sounds/Waves/VO/VO_S33_ZeroPointSprite_Inventory_Thrown_011}}
 {id=WavePlayer_6, properties {soundWaveAssetPtr=/Sprites_ZeroPointSprite/Assets/Sounds/Waves/VO/VO_S33_ZeroPointSprite_Inventory_Thrown_012}}
 {id=WavePlayer_7, properties {soundWaveAssetPtr=/Sprites_ZeroPointSprite/Assets/Sounds/Waves/VO/VO_S33_ZeroPointSprite_Inventory_Thrown_013}}
 {id=WavePlayer_9, properties {soundWaveAssetPtr=/Sprites_ZeroPointSprite/Assets/Sounds/Waves/VO/VO_S33_ZeroPointSprite_Inventory_Thrown_015}}
]
properties {
 attenuationSettings=/Game/Sounds/Attenuation/Fort_Weapon_Projectile_InAir_Loop_Att
 maxDistance=4500
 soundClassObject=/Game/Sounds/FortSoundClassesAndModes/Fort_Creatures_Vocalization_Normal
 volumeMultiplier=1
}
